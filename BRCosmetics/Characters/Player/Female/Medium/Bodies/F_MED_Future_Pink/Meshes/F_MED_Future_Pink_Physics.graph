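# Physics asset for the Future Pink female medium body cosmetic, defining secondary motion for its dangling armband, laces and shin guard straps.
# Small box bodies form jointed chains: a branching armband on the right upper arm, branching laces on the right calf, and two two-segment straps on the left calf.
# The laces hang from a master body with free rotation against the right calf's capsule, while the other joints carry angular limit values.
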PhysicsAsset /BRCosmetics/Characters/Player/Female/Medium/Bodies/F_MED_Future_Pink/Meshes/F_MED_Future_Pink_Physics
bodies {
 DYN_Armband [{shapeType=Box, extentX=2.75, extentY=0.625, extentZ=2.8125, center=<1.4634 0 0>}]
 DYN_ArmbandL1 [{shapeType=Box, extentX=2.625, extentY=1, extentZ=2.0625, center=<1.4626 0 0>}]
 DYN_ArmbandL2 [{shapeType=Box, extentX=3.25, extentY=1, extentZ=2.25, center=<1.8091 0 0>}]
 DYN_ArmbandR1 [{shapeType=Box, extentX=2.3125, extentY=1, extentZ=2.8125, center=<1.2817 0 0>}]
 DYN_ArmbandR2 [{shapeType=Box, extentX=2.375, extentY=1, extentZ=2.6875, center=<1.2866 0.0001 0>}]
 DYN_ArmbandR3 [{shapeType=Box, extentX=2.875, extentY=1, extentZ=2.5625, center=<1.5244 0 0.0001>}]
 DYN_Laces [{shapeType=Box, extentX=8.375, extentY=1, extentZ=8, center=<4.3036 0 0>, rotation=<rot 0 0 11.2523>}]
 DYN_LacesL1 [{shapeType=Box, extentX=3.3125, extentY=1, extentZ=3.8125, center=<1.7999 0 0>}]
 DYN_LacesL2 [{shapeType=Box, extentX=3.5625, extentY=1, extentZ=3.9375, center=<1.9075 0 0>}]
 DYN_LacesR1 [{shapeType=Box, extentX=2.875, extentY=1, extentZ=4.4375, center=<1.6425 -0.0001 -0.2922>, rotation=<rot 0 0 -14.064>}]
 DYN_LacesR2 [{shapeType=Box, extentX=3.9375, extentY=3.8125, extentZ=1, center=<1.9308 0.2374 0>, rotation=<rot 0 0 5.6259>}]
 DYN_Laces_Master [{shapeType=Box, extentX=1, extentY=1, extentZ=1}]
 DYN_ShinGuardStrapLower_1 [{shapeType=Box, extentX=2.625, extentY=2.5, extentZ=1, center=<1.2966 0 -0.0001>}]
 DYN_ShinGuardStrapLower_2 [{shapeType=Box, extentX=2.1875, extentY=2.5, extentZ=1, center=<1.2439 0 0>}]
 DYN_ShinGuardStrapUpper_1 [{shapeType=Box, extentX=3.3125, extentY=2.625, extentZ=1, center=<1.8564 0 0>}]
 DYN_ShinGuardStrapUpper_2 [{shapeType=Box, extentX=1.9375, extentY=2.625, extentZ=1, center=<1.1601 0 0>}]
 calf_twist_02_l [{shapeType=Box, extentX=1, extentY=1, extentZ=1}]
 calf_twist_02_r [{shapeType=Capsule, length=23.5352, radius=5.3306, center=<3.1994 2.2047 0.5431>, rotation=<rot -2.1306 -82.6984 -89.4666>}]
 upperarm_twist_01_r [{shapeType=Box, extentX=1, extentY=1, extentZ=1}]
}
constraints [
 {
  bone1Name=DYN_Laces_Master
  bone2Name=calf_twist_02_r
  swing1LimitDegrees=45
  swing1Motion=Free
  swing2LimitDegrees=45
  swing2Motion=Free
  twistLimitDegrees=45
  twistMotion=Free
 }
 {bone1Name=DYN_Laces, bone2Name=DYN_Laces_Master, swing1LimitDegrees=45, swing2LimitDegrees=45, twistLimitDegrees=5}
 {bone1Name=DYN_LacesR1, bone2Name=DYN_Laces, swing1LimitDegrees=45, swing2LimitDegrees=45, twistLimitDegrees=5}
 {bone1Name=DYN_LacesR2, bone2Name=DYN_LacesR1, swing1LimitDegrees=45, swing2LimitDegrees=45, twistLimitDegrees=5}
 {bone1Name=DYN_LacesL1, bone2Name=DYN_Laces, swing1LimitDegrees=45, swing2LimitDegrees=45, twistLimitDegrees=5}
 {bone1Name=DYN_LacesL2, bone2Name=DYN_LacesL1, swing1LimitDegrees=45, swing2LimitDegrees=45, twistLimitDegrees=5}
 {bone1Name=DYN_ShinGuardStrapLower_2, bone2Name=DYN_ShinGuardStrapLower_1, swing1LimitDegrees=5, swing2LimitDegrees=17, twistLimitDegrees=5}
 {bone1Name=DYN_ShinGuardStrapUpper_2, bone2Name=DYN_ShinGuardStrapUpper_1, swing1LimitDegrees=5, swing2LimitDegrees=17, twistLimitDegrees=5}
 {bone1Name=DYN_ShinGuardStrapLower_1, bone2Name=calf_twist_02_l, swing1LimitDegrees=5, swing2LimitDegrees=17, twistLimitDegrees=5}
 {bone1Name=DYN_ShinGuardStrapUpper_1, bone2Name=calf_twist_02_l, swing1LimitDegrees=5, swing2LimitDegrees=17, twistLimitDegrees=5}
 {bone1Name=DYN_ArmbandL2, bone2Name=DYN_ArmbandL1, swing1LimitDegrees=12, swing2LimitDegrees=12, twistLimitDegrees=5}
 {bone1Name=DYN_Armband, bone2Name=upperarm_twist_01_r, swing1LimitDegrees=10, swing2LimitDegrees=10, twistLimitDegrees=3}
 {bone1Name=DYN_ArmbandR1, bone2Name=DYN_Armband, swing1LimitDegrees=12, swing2LimitDegrees=12, twistLimitDegrees=5}
 {bone1Name=DYN_ArmbandR2, bone2Name=DYN_ArmbandR1, swing1LimitDegrees=12, swing2LimitDegrees=12, twistLimitDegrees=5}
 {bone1Name=DYN_ArmbandR3, bone2Name=DYN_ArmbandR2, swing1LimitDegrees=12, swing2LimitDegrees=12, twistLimitDegrees=5}
 {bone1Name=DYN_ArmbandL1, bone2Name=DYN_Armband, swing1LimitDegrees=12, swing2LimitDegrees=12, twistLimitDegrees=5}
]
