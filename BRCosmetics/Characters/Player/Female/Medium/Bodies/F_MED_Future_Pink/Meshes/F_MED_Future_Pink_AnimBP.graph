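# Animation Blueprint for the F_MED_Future_Pink female body cosmetic, copying the attached parent's pose and simulating secondary bones with RigidBody.
# Bool blends select corrective branches: crouching adjusts kneepad, glute and lace bones, while targeting constrains armband bones.
# Skydiving, parachute or emote states enable an extra three-bone armband constraint.
# The EventGraph maps head rotation relative to spine_05 into four alphas that drive additive collar bone adjustments.
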AnimBlueprint /BRCosmetics/Characters/Player/Female/Medium/Bodies/F_MED_Future_Pink/Meshes/F_MED_Future_Pink_AnimBP
target_skeleton=&/F_MED_Future_Pink_Skeleton
graphs [
 AnimGraph {nodes [
   n1 = RigidBody(ComponentPose=Local To Component(LocalPose=Copy Pose From Mesh() {bUseAttachedParent=true})) {
    bOverrideWorldGravity=true
    overridePhysicsAsset=&/F_MED_Future_Pink_Physics
    simulationSpace=BaseBoneSpace
    alphaBoolBlend {blendOption=Linear}
    baseBoneRef {boneName=root}
    componentLinearAccScale=<0.3 0.3 0.3>
    componentLinearVelScale=<0.6 0.6 0.6>
    overrideWorldGravity {z=-980}
   }
   n2 = RigidBody(ComponentPose=Local To Component(LocalPose=Copy Pose From Mesh() {bUseAttachedParent=true})) {
    bOverrideWorldGravity=true
    overridePhysicsAsset=&F_MED_Future_Pink_Physics
    simulationSpace=BaseBoneSpace
    alphaBoolBlend {blendOption=Linear}
    baseBoneRef {boneName=root}
    componentLinearAccScale=<0.3 0.3 0.3>
    componentLinearVelScale=<0.6 0.6 0.6>
    overrideWorldGravity {z=-980}
   }
   n3 = "Constraint - DYN_Armband (P)"(ComponentPose=@n2, Alpha=1, ConstraintWeights_0=1) {
    alphaBoolBlend {blendOption=Linear}
    alphaScaleBiasClamp {scale=0.23}
    boneToModify {boneName=DYN_Armband}
    constraintSetup [{targetBone {boneName=upperarm_twist_02_r}}]
    constraintWeights [1]
   }
   n4 = "Constraint - DYN_ArmbandR1 (P)"(ComponentPose=@n3, Alpha=1, ConstraintWeights_0=1) {
    alphaBoolBlend {blendOption=Linear}
    alphaScaleBiasClamp {scale=0.23}
    boneToModify {boneName=DYN_ArmbandR1}
    constraintSetup [{targetBone {boneName=upperarm_twist_02_r}}]
    constraintWeights [1]
   }
   n5 = "Constraint - DYN_ArmbandL1 (P)"(ComponentPose=@n4, Alpha=1, ConstraintWeights_0=1) {
    alphaBoolBlend {blendOption=Linear}
    alphaScaleBiasClamp {scale=0.23}
    boneToModify {boneName=DYN_ArmbandL1}
    constraintSetup [{targetBone {boneName=upperarm_twist_02_r}}]
    constraintWeights [1]
   }
   n6 = "Constraint - DYN_ArmbandR2 (P)"(ComponentPose=@n5, Alpha=1, ConstraintWeights_0=1) {
    alphaBoolBlend {blendOption=Linear}
    alphaScaleBiasClamp {scale=0.23}
    boneToModify {boneName=DYN_ArmbandR2}
    constraintSetup [{targetBone {boneName=upperarm_twist_02_r}}]
    constraintWeights [1]
   }
   n7 = "Constraint - DYN_ArmbandL2 (P)"(ComponentPose=@n6, Alpha=1, ConstraintWeights_0=1) {
    alphaBoolBlend {blendOption=Linear}
    alphaScaleBiasClamp {scale=0.23}
    boneToModify {boneName=DYN_ArmbandL2}
    constraintSetup [{targetBone {boneName=upperarm_twist_02_r}}]
    constraintWeights [1]
   }
   n8 = "Constraint - DYN_ArmbandR3 (P)"(ComponentPose=@n7, Alpha=1, ConstraintWeights_0=1) {
    alphaBoolBlend {blendOption=Linear}
    alphaScaleBiasClamp {scale=0.23}
    boneToModify {boneName=DYN_ArmbandR3}
    constraintSetup [{targetBone {boneName=upperarm_twist_02_r}}]
    constraintWeights [1]
   }
   n9 = Blend Poses by bool(
    BlendPose_0=Component To Local(ComponentPose=@n8)
    BlendPose_1=Component To Local(ComponentPose=@n1)
    BlendTime_1=0.2
    bActiveValue=Get bIsTargeting()
   ) {blendType=Linear}
   n10 = RigidBody(ComponentPose=Local To Component(LocalPose=Copy Pose From Mesh() {bUseAttachedParent=true})) {
    bOverrideWorldGravity=true
    overridePhysicsAsset=&F_MED_Future_Pink_Physics
    simulationSpace=BaseBoneSpace
    alphaBoolBlend {blendOption=Linear}
    baseBoneRef {boneName=root}
    componentLinearAccScale=<0.3 0.3 0.3>
    componentLinearVelScale=<0.6 0.6 0.6>
    overrideWorldGravity {z=-980}
   }
   n11 = "Transform (Modify) Bone - Bone: DYN_Kneepad"(ComponentPose=@n10, Alpha=1, Translation="-3,-1.3,0", Rotation="0,-36,1", Scale="1,1,1") {
    rotationMode=BMM_Additive
    rotationSpace=BCS_ParentBoneSpace
    translationMode=BMM_Additive
    translationSpace=BCS_ParentBoneSpace
    alphaBoolBlend {blendOption=Linear}
    boneToModify {boneName=DYN_Kneepad}
    rotation {roll=1, yaw=-36}
    translation=<-3 -1.3>
   }
   n12 = "Transform (Modify) Bone - Bone: dyn_glute_l"(ComponentPose=@n11, Alpha=0.5, Translation="-2,1.5,0", Rotation="0,0,0", Scale="1,1,1") {
    rotationMode=BMM_Additive
    rotationSpace=BCS_ParentBoneSpace
    translationMode=BMM_Additive
    translationSpace=BCS_ParentBoneSpace
    alphaBoolBlend {blendOption=Linear}
    boneToModify {boneName=dyn_glute_l}
    translation=<-2 1.5>
   }
   n13 = "Transform (Modify) Bone - Bone: DYN_KneeL"(ComponentPose=@n12, Alpha=1, Translation="0,0,0", Rotation="0,0,0", Scale="1.000,1.000,0.6") {scaleMode=BMM_Additive, scaleSpace=BCS_BoneSpace, alphaBoolBlend {blendOption=Linear}, boneToModify {boneName=DYN_KneeL}, scale {z=0.6}}
   n14 = "Constraint - DYN_LacesL1 (P)"(ComponentPose=@n13, Alpha=0.9, ConstraintWeights_0=1) {
    alphaBoolBlend {blendOption=Linear}
    alphaScaleBiasClamp {scale=0.4}
    boneToModify {boneName=DYN_LacesL1}
    constraintSetup [{targetBone {boneName=calf_r}}]
    constraintWeights [1]
   }
   n15 = "Constraint - DYN_LacesR1 (P)"(ComponentPose=@n14, Alpha=0.9, ConstraintWeights_0=1) {
    alphaBoolBlend {blendOption=Linear}
    alphaScaleBiasClamp {scale=0.4}
    boneToModify {boneName=DYN_LacesR1}
    constraintSetup [{targetBone {boneName=calf_r}}]
    constraintWeights [1]
   }
   n16 = "Constraint - DYN_LacesL2 (P)"(ComponentPose=@n15, Alpha=0.9, ConstraintWeights_0=1) {
    alphaBoolBlend {blendOption=Linear}
    alphaScaleBiasClamp {scale=0.4}
    boneToModify {boneName=DYN_LacesL2}
    constraintSetup [{targetBone {boneName=calf_r}}]
    constraintWeights [1]
   }
   n17 = "Constraint - DYN_LacesR2 (P)"(ComponentPose=@n16, Alpha=0.9, ConstraintWeights_0=1) {
    alphaBoolBlend {blendOption=Linear}
    alphaScaleBiasClamp {scale=0.4}
    boneToModify {boneName=DYN_LacesR2}
    constraintSetup [{targetBone {boneName=calf_r}}]
    constraintWeights [1]
   }
   n18 = Component To Local(ComponentPose=@n17)
   n19 = Get bIsCrouching()
   n20 = Get bIsSkyDiving()
   n21 = Get bIsParachuteOpened()
   n22 = Get bIsPlayingEmote()
   n23 = "Constraint - DYN_Armband (PPP)"!(
    ComponentPose=Local To Component(LocalPose=Blend Poses by bool(BlendPose_0=@n18, BlendPose_1=@n9, BlendTime_1=0.2, bActiveValue=@n19) {blendType=Linear})
    Alpha="To Float (Boolean)"(InBool=OR Boolean!(A=@n20, B=@n21, C:bool=@n22) -> (ReturnValue))
    ConstraintWeights_0=0.33
    ConstraintWeights_1:float=0.33
    ConstraintWeights_2:float=0.34
   ) -> (Pose) {
    alphaBoolBlend {blendOption=Linear}
    boneToModify {boneName=DYN_Armband}
    constraintSetup [{targetBone {boneName=upperarm_twist_01_r}}, {targetBone {boneName=upperarm_twist_02_r}}, {targetBone {boneName=upperarm_r}}]
    constraintWeights [1,1,1]
   }
   n24 = "Transform (Modify) Bone - Bone: DYN_CollarAdjust"(ComponentPose=@n23, Alpha=Get Head_Up() -> (Head_Up:double), Translation="0,0,0", Rotation="0.000,0.000,21", Scale="1,1,1.2") {
    rotationMode=BMM_Additive
    scaleMode=BMM_Additive
    scaleSpace=BCS_BoneSpace
    translationMode=BMM_Additive
    alphaBoolBlend {blendOption=Linear}
    boneToModify {boneName=DYN_CollarAdjust}
    rotation {roll=21}
    scale {z=1.2}
   }
   n25 = "Transform (Modify) Bone - Bone: DYN_CollarAdjust"(ComponentPose=@n24, Alpha=Get Head_Down() -> (Head_Down:double), Translation="0,0,0", Rotation="0,0,0", Scale="1,1,1.2") {
    rotationMode=BMM_Additive
    scaleMode=BMM_Additive
    scaleSpace=BCS_BoneSpace
    translationMode=BMM_Additive
    alphaBoolBlend {blendOption=Linear}
    boneToModify {boneName=DYN_CollarAdjust}
    scale {z=1.2}
   }
   n26 = "Transform (Modify) Bone - Bone: DYN_CollarAdjust"(ComponentPose=@n25, Alpha=Get Head_Left() -> (Head_Left:double), Translation="0,0,0", Rotation="0.000,20,0", Scale="1,1,1") {
    rotationMode=BMM_Additive
    translationMode=BMM_Additive
    alphaBoolBlend {blendOption=Linear}
    boneToModify {boneName=DYN_CollarAdjust}
    rotation {yaw=20}
   }
   n27 = "Transform (Modify) Bone - Bone: DYN_CollarAdjust"(ComponentPose=@n26, Alpha=Get Head_Right() -> (Head_Right:double), Translation="-0.7,0,0", Rotation="0.000,-20,0", Scale="1,1,1") {
    rotationMode=BMM_Additive
    translationMode=BMM_Additive
    alphaBoolBlend {blendOption=Linear}
    boneToModify {boneName=DYN_CollarAdjust}
    rotation {yaw=-20}
    translation {x=-0.7}
   }
   Output Pose(Result=Component To Local(ComponentPose=@n27)) {layerGroup=None}
  ]}
 EventGraph {nodes [
   Event BlueprintUpdateAnimation()
   > Sequence()
   n1 = GetOwnerSkeletalMeshComponent()
   n2 = BreakRotator(InRot=BreakTransform(InTransform=GetDeltaTransformFromRefPose(self=@n1, BoneName=head, BaseName=spine_05)).Rotation)
   > Set Head_Left(Head_Left:double=MapRangeClamped(Value=@n2, InRangeA=10, InRangeB=50)) -> (Output_Get:double)
   > Set Head_Right(Head_Right:double=MapRangeClamped(Value=@n2, InRangeA=0, InRangeB=-25)) -> (Output_Get:double)
   > Set Head_Up(Head_Up:double=MapRangeClamped(Value=@n2.Yaw, InRangeA=0, InRangeB=-45)) -> (Output_Get:double)
   > Set Head_Down(Head_Down:double=MapRangeClamped(Value=@n2.Yaw, InRangeA=0, InRangeB=45)) -> (Output_Get:double)
  ]}
]
variables [
 Head_Down {type=number}
 Head_Left {type=number}
 Head_Right {type=number}
 Head_Up {type=number}
]
